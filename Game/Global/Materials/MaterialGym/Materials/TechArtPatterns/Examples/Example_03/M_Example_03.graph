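# Material Gym tech-art example that builds an animated pattern from four colored layers of offset, tiled grid shapes.
# Each layer blends distance-field samples, scaled by a custom flipbook-UV HLSL node, then thresholds them with SmoothStep; two layers add a line-thickness control.
# The first layer also rotates over time with BumpOffset height.
# Sine-driven sphere masks provide morph masks, and the colored layers are combined additively or with the fourth overlapping.
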Material /Game/Global/Materials/MaterialGym/Materials/TechArtPatterns/Examples/Example_03/M_Example_03
graph {
 properties {bUsedWithMeshDeformer=true, bUsedWithSkeletalMesh=true, pixelDepthOffsetMode=PDOM_Legacy, usedShadingModels=Default Lit}
 nodes [
  n1 = LinearInterpolate(
   A=VectorParameter() {group=TechArtPattern, parameterName=Layer1_Color01, defaultValue=<rgba 0.3594 0 0.9167 1>}.RGB
   B=VectorParameter() {group=TechArtPattern, parameterName=Layer1_Color02, defaultValue=<rgba 1 0.2601 0 1>}.RGB
   Alpha=NamedRerouteUsage()
  )
  MaterialExpressionScalarParameter_20 = ScalarParameter() {defaultValue=2, group=TechArtPattern, parameterName=StartSize}
  MaterialExpressionScalarParameter_31 = ScalarParameter() {defaultValue=3, desc=2.0, group=TechArtPattern, parameterName=EndSize}
  MaterialExpressionConstant_15 = Constant() {r=2}
  MaterialExpressionNamedRerouteUsage_0 = NamedRerouteUsage()
  n2 = Constant() {r=5.5}
  n3 = MaterialFunctionCall!(Offset=Constant2Vector() {g=0.25, r=0.25}, UV, Tiling=@n2) {materialFunction=/Game/Global/Materials/MaterialGym/Materials/TechArtPatterns/MaterialFunctions/MF_GridUVs}
  n4 = Constant() {r=0.1}
  MaterialExpressionBumpOffset_0 = BumpOffset(
   Coordinate=MaterialFunctionCall(
    UVs=@n3.GridUV_Tiling
    "Rotation Angle (0-1)"=MaterialFunctionCall(Speed=@n4) {materialFunction=&TimeWithSpeedVariable}.Frac Time
   ) {materialFunction=&CustomRotator}.Rotated Values
   Height=ScalarParameter() {defaultValue=-1, group=TechArtPattern, parameterName=Height01}
  )
  n5 = Custom!(
   s_start=@MaterialExpressionScalarParameter_20
   s_end=@MaterialExpressionScalarParameter_31
   fs=@MaterialExpressionConstant_15
   time=@MaterialExpressionNamedRerouteUsage_0
   uv=@MaterialExpressionBumpOffset_0
  ) {
   code="// fs = flipbook size\r\nfloat l = lerp(s_start, s_end, time);\r\nreturn uv / l + (1/fs-1/l) / 2 ;"
   description=FlipbookScaleUV
   outputType=CMOT_Float2
   inputs [
    {
     inputName=s_start
     input {expression=&self:MaterialExpressionScalarParameter_20}
    }
    {
     inputName=s_end
     input {expression=&self:MaterialExpressionScalarParameter_31}
    }
    {
     inputName=fs
     input {expression=&self:MaterialExpressionConstant_15}
    }
    {
     inputName=time
     input {expression=&self:MaterialExpressionNamedRerouteUsage_0}
    }
    {
     inputName=uv
     input {expression=&self:MaterialExpressionBumpOffset_0}
    }
   ]
  }
  n6 = TextureSample(UVs=Add(A=@n5, B=Constant2Vector() {g=0.5})) {
   desc=circle
   samplerSource=SSM_Clamp_WorldGroupSettings
   samplerType=SAMPLERTYPE_LinearColor
   texture=/Game/Global/Materials/MaterialGym/Materials/TechArtPatterns/Textures/T_Shapes_DF
  }
  n7 = TextureSample(UVs=Add(A=@n5, B=Constant2Vector()) {constB=0.5}) {
   samplerSource=SSM_Clamp_WorldGroupSettings
   samplerType=SAMPLERTYPE_LinearColor
   texture=&T_Shapes_DF
  }
  n8 = NamedRerouteUsage()
  n9 = Distance(A=LinearInterpolate(A=@n6.R, B=@n7.R, Alpha=@n8) {constAlpha=1}, B=Constant() {r=0.65})
  n10 = NamedRerouteDeclaration(None=SmoothStep(Value=OneMinus(None=@n9)) {constMax=0.99, constMin=0.98}) {name=Shape00, nodeColor=<rgba 0 0.0156 1 1>}
  n11 = LinearInterpolate(
   A=VectorParameter() {group=TechArtPattern, parameterName=Layer2_Color01, defaultValue=<rgba 0.1634 0 0.4167 1>}.RGB
   B=VectorParameter() {group=TechArtPattern, parameterName=Layer2_Color02, defaultValue=<rgba 0.0317 0.3906 0 1>}.RGB
   Alpha=NamedRerouteUsage()
  )
  n12 = Constant() {r=1}
  n13 = Subtract(A=@n12, B=ScalarParameter() {defaultValue=0.5, group=TechArtPattern, parameterName=LineThickness})
  MaterialExpressionMaterialFunctionCall_4 = MaterialFunctionCall(Offset=Constant2Vector() {g=0.25, r=0.75}, Tiling=@n2) {materialFunction=&MF_GridUVs}
  n14 = TextureSample(UVs=@MaterialExpressionMaterialFunctionCall_4.GridUV_Tiling) {
   samplerSource=SSM_Clamp_WorldGroupSettings
   samplerType=SAMPLERTYPE_LinearColor
   texture=&T_Shapes_DF
  }
  n15 = Custom(
   s_start=@MaterialExpressionScalarParameter_20
   s_end=@MaterialExpressionScalarParameter_31
   fs=@MaterialExpressionConstant_15
   time=@MaterialExpressionNamedRerouteUsage_0
   uv=@MaterialExpressionMaterialFunctionCall_4.GridUV_Tiling
  ) {
   code="// fs = flipbook size\r\nfloat l = lerp(s_start, s_end, time);\r\nreturn uv / l + (1/fs-1/l) / 2 ;"
   description=FlipbookScaleUV
   outputType=CMOT_Float2
   inputs [
    {
     inputName=s_start
     input {expression=&self:MaterialExpressionScalarParameter_20}
    }
    {
     inputName=s_end
     input {expression=&self:MaterialExpressionScalarParameter_31}
    }
    {
     inputName=fs
     input {expression=&self:MaterialExpressionConstant_15}
    }
    {
     inputName=time
     input {expression=&self:MaterialExpressionNamedRerouteUsage_0}
    }
    {
     inputName=uv
     input {expression=&self:MaterialExpressionMaterialFunctionCall_4}
    }
   ]
  }
  n16 = TextureSample(UVs=Add(A=@n15, B=Constant2Vector()) {constB=0.5}) {
   samplerSource=SSM_Clamp_WorldGroupSettings
   samplerType=SAMPLERTYPE_LinearColor
   texture=&T_Shapes_DF
  }
  n17 = OneMinus(None=Distance(A=LinearInterpolate(A=@n14.R, B=@n16.R, Alpha=@n8) {constAlpha=1}, B=Constant() {r=0.65}))
  n18 = NamedRerouteDeclaration(None=SmoothStep(Value=SmoothStep(Min=@n13, Max=@n12, Value=@n17) {constMin=0.5}) {constMax=0.99, constMin=0.98}) {name=Shape01, nodeColor=<rgba 0 0.0625 1 1>}
  n19 = VectorParameter() {group=TechArtPattern, parameterName=Layer3_Color01, defaultValue=<rgba 0 0.3594 0.0062 1>}
  n20 = VectorParameter() {group=TechArtPattern, parameterName=Layer3_Color02, defaultValue=<rgba 0.2204 0 0.4167 1>}
  n21 = NamedRerouteUsage()
  n22 = Constant() {r=1}
  n23 = ScalarParameter() {defaultValue=0.5, group=TechArtPattern, parameterName=LineThickness}
  MaterialExpressionMaterialFunctionCall_5 = MaterialFunctionCall(Offset=Constant2Vector() {g=0.75, r=0.25}, Tiling=@n2) {materialFunction=&MF_GridUVs}
  n24 = TextureSample(UVs=@MaterialExpressionMaterialFunctionCall_5.GridUV_Tiling) {
   samplerSource=SSM_Clamp_WorldGroupSettings
   samplerType=SAMPLERTYPE_LinearColor
   texture=&T_Shapes_DF
  }
  n25 = Custom(
   s_start=@MaterialExpressionScalarParameter_20
   s_end=@MaterialExpressionScalarParameter_31
   fs=@MaterialExpressionConstant_15
   time=@MaterialExpressionNamedRerouteUsage_0
   uv=@MaterialExpressionMaterialFunctionCall_5.GridUV_Tiling
  ) {
   code="// fs = flipbook size\r\nfloat l = lerp(s_start, s_end, time);\r\nreturn uv / l + (1/fs-1/l) / 2 ;"
   description=FlipbookScaleUV
   outputType=CMOT_Float2
   inputs [
    {
     inputName=s_start
     input {expression=&self:MaterialExpressionScalarParameter_20}
    }
    {
     inputName=s_end
     input {expression=&self:MaterialExpressionScalarParameter_31}
    }
    {
     inputName=fs
     input {expression=&self:MaterialExpressionConstant_15}
    }
    {
     inputName=time
     input {expression=&self:MaterialExpressionNamedRerouteUsage_0}
    }
    {
     inputName=uv
     input {expression=&self:MaterialExpressionMaterialFunctionCall_5}
    }
   ]
  }
  n26 = TextureSample(UVs=Add(A=@n25, B=Constant2Vector()) {constB=0.5}) {
   samplerSource=SSM_Clamp_WorldGroupSettings
   samplerType=SAMPLERTYPE_LinearColor
   texture=&T_Shapes_DF
  }
  n27 = Distance(A=LinearInterpolate(A=@n24.R, B=@n26.R, Alpha=@n8) {constAlpha=1}, B=Constant() {r=0.65})
  n28 = SmoothStep(Value=SmoothStep(Min=Subtract(A=@n22, B=@n23), Max=@n22, Value=OneMinus(None=@n27)) {constMin=0.5}) {constMax=0.99, constMin=0.98}
  n29 = Add(
   A=Add(A=Multiply(A=@n1, B=@n10), B=Multiply(A=@n11, B=@n18))
   B=Multiply(A=LinearInterpolate(A=@n19.RGB, B=@n20.RGB, Alpha=@n21), B=NamedRerouteDeclaration(None=@n28) {name=Shape10, nodeColor=<rgba 0 0.0625 1 1>})
  )
  n30 = LinearInterpolate(
   A=VectorParameter() {group=TechArtPattern, parameterName=Layer4_Color01, defaultValue=<rgba 1.6122 0 5 1>}.RGB
   B=VectorParameter() {group=TechArtPattern, parameterName=Layer4_Color02, defaultValue=<rgba 0.0293 0 0.1771 1>}.RGB
   Alpha=NamedRerouteUsage()
  )
  MaterialExpressionMaterialFunctionCall_6 = MaterialFunctionCall(Offset=Constant2Vector() {g=0.75, r=0.75}, Tiling=@n2) {materialFunction=&MF_GridUVs}
  n31 = TextureSample(UVs=@MaterialExpressionMaterialFunctionCall_6.GridUV_Tiling) {
   samplerSource=SSM_Clamp_WorldGroupSettings
   samplerType=SAMPLERTYPE_LinearColor
   texture=/Game/Global/Materials/MaterialGym/Materials/TechArtPatterns/Textures/T_Circle_DF
  }
  n32 = Custom(
   s_start=@MaterialExpressionScalarParameter_20
   s_end=@MaterialExpressionScalarParameter_31
   fs=@MaterialExpressionConstant_15
   time=@MaterialExpressionNamedRerouteUsage_0
   uv=@MaterialExpressionMaterialFunctionCall_6.GridUV_Tiling
  ) {
   code="// fs = flipbook size\r\nfloat l = lerp(s_start, s_end, time);\r\nreturn uv / l + (1/fs-1/l) / 2 ;"
   description=FlipbookScaleUV
   outputType=CMOT_Float2
   inputs [
    {
     inputName=s_start
     input {expression=&self:MaterialExpressionScalarParameter_20}
    }
    {
     inputName=s_end
     input {expression=&self:MaterialExpressionScalarParameter_31}
    }
    {
     inputName=fs
     input {expression=&self:MaterialExpressionConstant_15}
    }
    {
     inputName=time
     input {expression=&self:MaterialExpressionNamedRerouteUsage_0}
    }
    {
     inputName=uv
     input {expression=&self:MaterialExpressionMaterialFunctionCall_6}
    }
   ]
  }
  n33 = TextureSample(UVs=Add(A=@n32, B=Constant2Vector() {r=0.5}) {constB=0.5}) {
   samplerSource=SSM_Clamp_WorldGroupSettings
   samplerType=SAMPLERTYPE_LinearColor
   texture=&T_Shapes_DF
  }
  n34 = Distance(A=LinearInterpolate(A=@n31.R, B=@n33.R, Alpha=@n8) {constAlpha=1}, B=Constant() {r=0.65})
  n35 = NamedRerouteDeclaration(None=SmoothStep(Value=OneMinus(None=@n34)) {constMax=0.99, constMin=0.98}) {name=Shape11, nodeColor=<rgba 0 0.0625 1 1>}
  Add(A=@n29, B=Multiply(A=@n30, B=@n35))
  Add(A=@n32, B=Constant2Vector() {g=0.5})
  Add(A=@n15, B=Constant2Vector() {g=0.5})
  Add(A=@n25, B=Constant2Vector() {g=0.5})
  n36 = Multiply(A=Time(), B=ScalarParameter() {defaultValue=0.2, group=TechArtPattern, parameterName=MorphSpeed}) {constB=0.5}
  Frac(None=@n36)
  n37 = ConstantBiasScale(None=Sine(None=@n36))
  n38 = LinearInterpolate(
   A=SphereMask(A=NamedRerouteUsage(), B=Constant2Vector() {g=0.5, r=0.5}, Radius=LinearInterpolate(Alpha=@n37) {constB=2}) {attenuationRadius=0.5, hardnessPercent=3}
   Alpha=@n37
  )
  NamedRerouteDeclaration(None=Saturate(None=Power(Base=@n38))) {name=MorphMaskGrid02, nodeColor=<rgba 0 1 0.25 1>}
  n39 = LinearInterpolate(
   A=SphereMask(A=NamedRerouteUsage(), B=Constant2Vector() {g=0.5, r=0.5}, Radius=LinearInterpolate(Alpha=@n37) {constB=2}) {attenuationRadius=0.5, hardnessPercent=3}
   Alpha=@n37
  )
  NamedRerouteDeclaration(None=Saturate(None=Power(Base=@n39))) {name=MorphMaskGrid03, nodeColor=<rgba 0 1 0.25 1>}
  n40 = LinearInterpolate(
   A=SphereMask(A=NamedRerouteUsage(), B=Constant2Vector() {g=0.5, r=0.5}, Radius=LinearInterpolate(Alpha=@n37) {constB=2}) {attenuationRadius=0.5, hardnessPercent=3}
   Alpha=@n37
  )
  NamedRerouteDeclaration(None=Saturate(None=Power(Base=@n40))) {name=MorphMaskGrid04, nodeColor=<rgba 0 1 0.25 1>}
  NamedRerouteDeclaration(None=@n37) {name=FracTime, nodeColor=<rgba 0 1 0.4844 1>}
  n41 = LinearInterpolate(
   A=SphereMask(A=TextureCoordinate(), B=Constant2Vector() {g=0.5, r=0.5}, Radius=LinearInterpolate(Alpha=@n37) {constB=2}) {attenuationRadius=0.5, hardnessPercent=3}
   Alpha=@n37
  )
  NamedRerouteDeclaration(None=Saturate(None=Power(Base=@n41))) {name=MorphMask, nodeColor=<rgba 0.3125 1 0 1>}
  NamedRerouteDeclaration(None=LinearInterpolate(A=@n29, B=@n30, Alpha=@n35)) {name=OverlappingPatterns, nodeColor=<rgba 1 0 0.1406 1>}
  NamedRerouteDeclaration(None=@n3.GridUV) {name=GridUV1, nodeColor=<rgba 1 0 0.0938 1>}
  NamedRerouteDeclaration(None=@MaterialExpressionMaterialFunctionCall_4.GridUV) {name=GridUV2, nodeColor=<rgba 1 0 0.0937 1>}
  NamedRerouteDeclaration(None=@MaterialExpressionMaterialFunctionCall_5.GridUV) {name=GridUV3, nodeColor=<rgba 1 0 0.0938 1>}
  NamedRerouteDeclaration(None=@MaterialExpressionMaterialFunctionCall_6.GridUV) {name=GridUV4, nodeColor=<rgba 1 0 0.0938 1>}
  n42 = LinearInterpolate(
   A=SphereMask(A=NamedRerouteUsage(), B=Constant2Vector() {g=0.5, r=0.5}, Radius=LinearInterpolate(Alpha=@n37) {constB=2}) {attenuationRadius=0.5, hardnessPercent=3}
   Alpha=@n37
  )
  NamedRerouteDeclaration(None=Saturate(None=Power(Base=@n42))) {name=MorphMaskGrid01, nodeColor=<rgba 0 1 0.25 1>}
  NamedRerouteUsage()
 ]
}
